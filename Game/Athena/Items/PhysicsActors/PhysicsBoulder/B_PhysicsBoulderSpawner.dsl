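; B_PhysicsBoulderSpawner is a spawner Blueprint for the physics boulder actor that shows a preview of what it spawns.
; Its construction script, only when running in the editor or PIE, gets the actor class to spawn and assigns it to the PrevisActor child actor component.
; The file contains only this previsualization logic, with no runtime spawning graph.
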(fn ConstructionScript
  (if (Editor|InEditororPIE)
    (bind _returnvalue (CallFunction|GetActorClasstoSpawn))
    (ChildActorComponent|SetChildActorClass (Variables|Default|GetPrevisActor) _returnvalue)))
