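; Desert variant of the physics boulder prop, adding Niagara effects on top of its parent's behavior, each flagged with User.DesertBoulder.
; BeginPlay attaches a motion effect to the static mesh.
; Hits at or above a speed threshold spawn a rolling effect at the impact point, only off the server, fed the mesh velocity.
; Death spawns a destroyed effect at the mesh location, passing mesh velocity as momentum rather than the event's Momentum.
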(event EventOnDeathPlayEffects
  (params (Damage :type Float%20%28single-precision%29) (DamageTags :type Gameplay%20Tag%20Container%20Structure%20%28by%20ref%29) (Momentum :type Vector :default "0, 0, 0") (HitInfo :type Hit%20Result%20Structure%20%28by%20ref%29) (InstigatedBy :type Fort%20Pawn%20Object%20Reference) (DamageCauser :type Actor%20Object%20Reference) (EffectContext :type Gameplay%20Effect%20Context%20Handle%20Structure))
  (bind _staticmeshcomponent (Variables|Debug|GetStaticMeshComponent))
  (|CallParentFunction Damage DamageTags Momentum HitInfo InstigatedBy DamageCauser EffectContext)
  (bind _returnvalue (Niagara|SpawnSystematLocation "/Game/Environments/Apollo/Rocks/PhysicsBoulder/NS_Physics_Boulder_Destroyed.NS_Physics_Boulder_Destroyed" (Transformation|GetWorldLocation _staticmeshcomponent)))
  (Niagara|SetNiagaraVariable(Vector3) _returnvalue "User.Momentum" (Transformation|GetComponentVelocity _staticmeshcomponent))
  (Niagara|SetNiagaraVariable(Bool) _returnvalue "User.DesertBoulder" true))

(event Collision|EventHit
  (params (MyComp :type Primitive%20Component%20Object%20Reference) (Other :type Actor%20Object%20Reference) (OtherComp :type Primitive%20Component%20Object%20Reference) (bSelfMoved :type Boolean) (HitLocation :type Vector :default "0, 0, 0") (HitNormal :type Vector :default "0, 0, 0") (NormalImpulse :type Vector :default "0, 0, 0") (Hit :type Hit%20Result%20Structure%20%28by%20ref%29))
  (|CallParentFunction MyComp Other OtherComp bSelfMoved HitLocation HitNormal NormalImpulse Hit)
  (if (< (Math|Vector|VectorLength (Transformation|GetComponentVelocity (Variables|Debug|GetStaticMeshComponent))) 3e+02)
    _
    (elif (Networking|IsServer)
      _
      (else
        (bind _returnvalue (Niagara|SpawnSystematLocation "/Game/Environments/Apollo/Rocks/PhysicsBoulder/NS_Physics_Boulder_Rolling.NS_Physics_Boulder_Rolling" HitLocation))
        (Niagara|SetNiagaraVariable(Vector3) _returnvalue "User.PhysicsVelocity" (Transformation|GetComponentVelocity (Variables|Debug|GetStaticMeshComponent)))
        (Niagara|SetNiagaraVariable(Bool) _returnvalue "User.DesertBoulder" true)))))

(event EventBeginPlay
  (|CallParentFunction)
  (bind _returnvalue (Niagara|SpawnSystemAttached "/Game/Environments/Apollo/Rocks/PhysicsBoulder/NS_Physics_Boulder_Motion.NS_Physics_Boulder_Motion" (Variables|Debug|GetStaticMeshComponent)))
  (Niagara|SetNiagaraVariable(Bool) _returnvalue "User.DesertBoulder" true))

(fn ConstructionScript
  (|CallParentFunction))
