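; Parent physics boulder prop that launches player pawns it hits, based on its own speed at impact.
; On server hit events with active physics, speed is matched against scalable impact tiers to pick an impulse.
; Launch direction combines the inverted horizontal impact normal with a vertical ratio.
; Hits from a player riding the mesh instead set a launch delay, preventing accidental launches during tricks like riding in water.
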(event EventBeginPlay
  (|CallParentFunction)
  (Collision|AssignOnComponentHit (Variables|Debug|GetStaticMeshComponent) (delegate Custom|ComponentHit)))

(event Custom|ComponentHit
  (params (HitComponent :type Primitive%20Component%20Object%20Reference) (OtherActor :type Actor%20Object%20Reference) (OtherComp :type Primitive%20Component%20Object%20Reference) (NormalImpulse :type Vector) (Hit :type Hit%20Result%20Structure%20%28by%20ref%29))
  (bind _returnvalue (Transformation|GetVelocity))
  (bind _returnvalue_2 (Math|Vector|VectorLength _returnvalue))
  (bind (_bblockinghit _binitialoverlap _time _distance _location _impactpoint _normal _impactnormal) (Collision|BreakHitResult Hit))
  (bind (_minimum_met _impulse) (CallFunction|GetPlayerImpulseVelocityfromImpactVelocity _returnvalue_2))
  (bind _returnvalue_3 (Utilities|Time|GetTimeSeconds))
  (bind _nextpotentiallaunchtime (Variables|Default|GetNextPotentialLaunchTime))
  (bind _returnvalue_4 (> _returnvalue_3 _nextpotentiallaunchtime))
  (if (Networking|IsServer)
    (Variables|Default|SetLastImpactNormal _impactnormal)
    (bind _returnvalue_5 (BuildingActor|PhysicsSimulationSupport|GetPhysicsObjectComponent self))
    (if (Components|Activation|IsActive _returnvalue_5)
      (bind _asfort_player_pawn (Utilities|Casting|CastToFortPlayerPawn OtherActor))
      (if (== (Variables|Debug|GetStaticMeshComponent) (Pawn|Components|CharacterMovement|GetMovementBaseObject (Variables|/Script/Engine.Character|GetCharacterMovement _asfort_player_pawn)))
        ;[1 1 1] Set a launch delay for certain interactions, like riding the log in water. This will prevent the player from accidentally being launched when performing cool water tricks and jumps
        (if _returnvalue_4
          (Variables|Default|SetNextPotentialLaunchTime (+ (Utilities|Time|GetTimeSeconds) (Variables|Default|GetLaunchDelay))))
        (elif _minimum_met
          (if _returnvalue_4
            (Character|LaunchCharacter _asfort_player_pawn (* (Math|Vector|Normalize (+ (* (Variables|Default|GetLastImpactNormal) "-1.000000,-1.000000,0.000000") (Math|Vector|MakeVector 0.0 0.0 (ScalableFloat|GetValueatLevel (Variables|Default|GetVerticalImpulseRatio))))) _impulse) true true)))))))

(fn GetPlayerImpulseVelocityFromImpactVelocity
  (params (Impact :type Float%20%28double-precision%29))
  (bind (_arrayindex _arrayelement) (Utilities|Array|ReverseforEachLoop (Variables|Default|GetPlayerImpactTiers))
    (:LoopBody
      (if (>= (Variables|GetImpact) (ScalableFloat|GetValueatLevel _arrayelement))
        (LocalVariables|SetReturnImpulseValue (ScalableFloat|GetValueatLevel (Utilities|Array|Get(acopy) (Variables|Default|GetPlayerImpulseTiers) _arrayindex)))))
    (:Completed
      (return (>= (LocalVariables|GetReturnImpulseValue) 0) (LocalVariables|GetReturnImpulseValue)))))

(fn ConstructionScript)
